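# Icon widget for match objective (urgent quest) entries in the HUD, extending AthenaSpecialEventEntryIcon.
# SetIconStyling stores quest data, loads the icon and tints the background's dynamic material by style tag.
# An async-loaded match quest category icon takes priority; otherwise AlertIcon validity toggles image versus "!" text.
# Functions set XP or gold rewards and play or reverse outro animations: success shows a glowing checkmark, failure has no tracks.
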WidgetBlueprint /Game/Athena/HUD/Discoverability/MatchObjectivesWidgets/WBP_Athena_Objectives_Icon
animations [
 Anim_Outro_Success {
  binding_count=3
  length_seconds=1.4
  object=&self:Anim_Outro_Success
  sequence {bindings [
    Image_CheckmarkGlow {
     id=#1
     possessed_class=Image
     tracks [
      {
       type=MovieSceneColorTrack
       display_name=Color and Opacity
       sections [{
         type=MovieSceneColorSection
         end_frame=28
         start_frame=0
         channels [
          Color.R {type=MovieSceneScriptingFloatChannel, const [0,2]}
          Color.G {type=MovieSceneScriptingFloatChannel, const [1,2]}
          Color.B {type=MovieSceneScriptingFloatChannel, const [0.1441,3]}
          Color.A {type=MovieSceneScriptingFloatChannel, const [1,2]}
         ]
        }]
      }
      {
       type=MovieSceneFloatTrack
       display_name=Render Opacity
       sections [{
         type=MovieSceneFloatSection
         end_frame=28
         start_frame=0
         channels [None {type=MovieSceneScriptingFloatChannel, keys [[0,0],[9,0],[11,1],[16,0]]}]
        }]
      }
     ]
    }
    Image_CheckMark {
     id=#2
     possessed_class=Image
     tracks [
      {
       type=MovieSceneFloatTrack
       display_name=Render Opacity
       sections [{
         type=MovieSceneFloatSection
         end_frame=28
         start_frame=0
         channels [None {type=MovieSceneScriptingFloatChannel, thinned_from=5, keys [[0,0],[7,0],[11,1],[16,1]]}]
        }]
      }
      {
       type=MovieScene2DTransformTrack
       sections [{
         type=MovieScene2DTransformSection
         end_frame=28
         start_frame=0
         channels [
          Translation.X {type=MovieSceneScriptingFloatChannel}
          Translation.Y {type=MovieSceneScriptingFloatChannel}
          Angle {type=MovieSceneScriptingFloatChannel}
          Scale.X {type=MovieSceneScriptingFloatChannel, keys [[5,3],[11,1],[16,1]]}
          Scale.Y {type=MovieSceneScriptingFloatChannel, keys [[5,3],[11,1],[16,1]]}
          Shear.X {type=MovieSceneScriptingFloatChannel, const [0,1]}
          Shear.Y {type=MovieSceneScriptingFloatChannel, const [0,1]}
         ]
        }]
      }
     ]
    }
    Image_Spark {
     id=#3
     possessed_class=Image
     tracks [
      {
       type=MovieSceneFloatTrack
       display_name=Render Opacity
       sections [{
         type=MovieSceneFloatSection
         end_frame=28
         start_frame=0
         channels [None {type=MovieSceneScriptingFloatChannel, keys [[0,0],[9,0],[11,1],[16,0]]}]
        }]
      }
      {
       type=MovieScene2DTransformTrack
       sections [{
         type=MovieScene2DTransformSection
         end_frame=28
         start_frame=9
         channels [
          Translation.X {type=MovieSceneScriptingFloatChannel}
          Translation.Y {type=MovieSceneScriptingFloatChannel}
          Angle {type=MovieSceneScriptingFloatChannel}
          Scale.X {type=MovieSceneScriptingFloatChannel, keys [[9,2.5],[16,1]]}
          Scale.Y {type=MovieSceneScriptingFloatChannel, keys [[9,2.5],[16,1]]}
          Shear.X {type=MovieSceneScriptingFloatChannel}
          Shear.Y {type=MovieSceneScriptingFloatChannel}
         ]
        }]
      }
     ]
    }
   ]}
 }
 Anim_Outro_Failed {
  binding_count=0
  length_seconds=1.1566
  object=&self:Anim_Outro_Failed
  sequence {bindings []}
 }
]
graphs [
 Check for Match Quest Icon {nodes [
   Check for Match Quest Icon() {is=title K2Node_FunctionEntry}
   n1 = Break Urgent Quest Data!(UrgentQuestData:"Urgent Quest Data Structure (by ref)"=Get Urgent Quest Data() -> (Urgent Quest Data:Urgent Quest Data Structure)) -> (
    EventTag:Gameplay Tag Structure
    VisualStyleTag:Gameplay Tag Container Structure
    TotalEventTime:float=~
    bDisplayTimer:bool=~
    bQuestCompleteOnTimeExpiration:bool=~
    bQuickIntro:bool=~
    bQuickOutro:bool=~
    bRewardPersist:bool=~
    bDisplayMultiStepExpanded:bool=~
    EventMainTitle:Text
    EventTitle:Text
    EventSubtitle:Text
    EventSubtitleSecondary:Text
    EventFailureMessage:Text
    EventDescription:Text
    EventDescriptionGamepad:Text
    EventDescriptionTouch:Text
    EventDescriptionOverrides:Array of Urgent Quest Data Desc Overrides Structures
    EventDescriptionOverrideTag:Gameplay Tag Structure
    SortPriority:int
    bHideQuestProgressBar:bool=~
    bDisplayAllObjectivesIndividually:bool=~
    QuestProgressTextFormat:Text
    bReverseProgressBar:bool=~
    bShowBountyThreatInformation:bool=~
    bShowBountyPriceAndDistance:bool=~
    bDisplayStartedEventUI:bool=~
    bIsSharedUrgentQuest:bool=~
    DisplayPlayer:Fort Player State Athena Object Reference
    AcceptingPlayer:Fort Player State Athena Object Reference
    SocialAvatarBrushPtr:Texture 2D Soft Object Reference
    BountyPriceImage:Object Reference
    AlertIcon:Soft Object Reference
    CharacterIcon:Soft Object Reference
    SpecialEventStartedSound:Sound Base Soft Object Reference
    AlertIconBorderColor:Linear Color Structure
    AlertIconBrushSize:Vector 2D Structure
    bOverrideWidgetPrimaryColor:bool=~
    OverriddenWidgetPrimaryColor:Linear Color Structure
    ObjectiveCategory:EObjectiveCategories Enum
   ) {is=title K2Node_BreakStruct}
   > n2 = Branch(Condition="Equal (GameplayTag)"(A=Get MatchQuestTag() -> (MatchQuestTag:Gameplay Tag Structure), B=@n1))
   > Return Node!(
    execute
    MatchQuestIcon:Texture 2D Soft Object Reference=Get Match Quest Category Icon!(Event Subtitle:"Text (by ref)"=@n1.EventSubtitle) -> (Category Icon:Texture 2D Soft Object Reference) {is=title K2Node_CallFunction}
   )
   Return Node(execute=@n2.else)
  ]}
 EventGraph {nodes [
   n1 = SetIconStyling!() -> (OutputDelegate, then, UrgentQuestData:"Urgent Quest Data Structure (by ref)") {is=title K2Node_CustomEvent}
   > n2 = Set Urgent Quest Data(Urgent Quest Data:Urgent Quest Data Structure=@n1.UrgentQuestData) -> (Output_Get:Urgent Quest Data Structure)
   > LoadAndSetIcon!(execute, self:Athena Special Event Entry Icon Object Reference, UrgentQuestData:"Urgent Quest Data Structure (by ref)"=@n2) -> (then, ReturnValue:bool)
   > Set StyleType(StyleType:Name=GetStylingTypeFromTag!(UrgentQuestData:Urgent Quest Data Structure=@n2) -> (OutStyleType:Name=None, IsABounty:bool)) -> (Output_Get:Name=None)
   > n3 = GetDynamicMaterial(self=Get Image_RewardBG() -> (Image_RewardBG:Image Object Reference))
   n4 = GetIconBGColors!(Type:Name=Get StyleType() -> (StyleType:Name=None)) -> (Icon BG Color 01:Linear Color Structure, Icon BG Color 02:Linear Color Structure)
   > SetVectorParameterValue(self=@n3, ParameterName=BG_Color1, Value=@n4)
   > SetVectorParameterValue(self=@n3, ParameterName=BG_Color2, Value=@n4.Icon BG Color 02)
   > n5 = Check for Match Quest Icon!(execute, self) -> (then, MatchQuestIcon:Texture 2D Soft Object Reference) {is=title K2Node_CallFunction}
   > n6 = Branch(Condition=IsValidSoftObjectReference(SoftObjectReference=@n5))
   > SetVisibility(self=Get Image_UrgentIcon() -> (Image_UrgentIcon:Image Object Reference), InVisibility=SelfHitTestInvisible)
   > SetVisibility(self=Get Text_UrgentIcon() -> (Text_UrgentIcon:Common Text Object Reference), InVisibility=Collapsed)
   > n7 = Async Load Asset(Asset=@n5)
   n8 = Cast To Texture2D(execute=@n7.Completed, Object=@n7)
   > SetBrushFromTexture!(execute, self:Image Object Reference=Get Image_UrgentIcon(), Texture:Texture 2D Object Reference=@n8, bMatchSize:bool=false) -> (then)
   n9 = Break Urgent Quest Data(UrgentQuestData=Get Urgent Quest Data()) -> (
    TotalEventTime=~
    bDisplayTimer=~
    bQuestCompleteOnTimeExpiration=~
    bQuickIntro=~
    bQuickOutro=~
    bRewardPersist=~
    bDisplayMultiStepExpanded=~
    bHideQuestProgressBar=~
    bDisplayAllObjectivesIndividually=~
    bReverseProgressBar=~
    bShowBountyThreatInformation=~
    bShowBountyPriceAndDistance=~
    bDisplayStartedEventUI=~
    bIsSharedUrgentQuest=~
    bOverrideWidgetPrimaryColor=~
   )
   n10 = IsValidSoftObjectReference(SoftObjectReference=@n9.AlertIcon)
   SetVisibility(execute=@n6.else, self=Get Image_UrgentIcon(), InVisibility=Select(Option 0=Hidden, Option 1=SelfHitTestInvisible, Index=@n10))
   > SetVisibility(self=Get Text_UrgentIcon(), InVisibility=Select(Option 0=SelfHitTestInvisible, Option 1=Collapsed, Index=@n10))
   > n11 = Branch(Condition=@n10)
   n12 = Async Load Asset(execute=@n11.else, Asset=@n9.AlertIcon)
   n13 = Cast To Texture2D(execute=@n12.Completed, Object=@n12)
   > SetBrushFromTexture(self=Get Image_UrgentIcon(), Texture=@n13, bMatchSize=false)
  ]}
 HideRewardText {nodes [
   n1 = HideRewardText!() -> (then, HideReward:bool=~) {is=title K2Node_FunctionEntry}
   > SetVisibility(
    self=Get Text_RewardQuantity() -> (Text_RewardQuantity:Fort Rich Text Block Keybind Object Reference)
    InVisibility=Select(Option 0=SelfHitTestInvisible, Option 1=Collapsed, Index=@n1)
   )
  ]}
 PlayOutroFailedAnim {nodes [
   PlayOutroFailedAnim() {is=title K2Node_FunctionEntry}
   > PlayAnimation(InAnimation=Get Anim_Outro_Failed() -> (Anim_Outro_Failed:Widget Animation Object Reference))
  ]}
 PlayOutroSuccessAnim {nodes [
   PlayOutroSuccessAnim() {is=title K2Node_FunctionEntry}
   > PlayAnimation(InAnimation=Get Anim_Outro_Success() -> (Anim_Outro_Success:Widget Animation Object Reference))
  ]}
 ResetCompletedAnims {nodes [
   n1 = ResetCompletedAnims!() -> (then, IsQuestCompleted:bool=~) {is=title K2Node_FunctionEntry}
   n2 = Select(
    Option 0:Widget Animation Object Reference=Get Anim_Outro_Failed()
    Option 1:Widget Animation Object Reference=Get Anim_Outro_Success()
    Index=@n1
   ) -> (ReturnValue:Widget Animation Object Reference)
   > PlayAnimation(InAnimation=@n2, StartAtTime=GetEndTime(self=@n2), PlayMode=Reverse)
  ]}
 Set Reward Quantity and Type {nodes [
   n1 = Set Reward Quantity and Type!() -> (then, InText:Text, InType:EFortEarnedSubRewardType Enum=XP) {is=title K2Node_FunctionEntry}
   n2 = "To Text (Integer)"(Value=0)
   n3 = Get Image_UrgentIcon()
   > SetVisibility(
    self=@n3
    InVisibility=Select(
     Option 0:ESlateVisibility Enum=SelfHitTestInvisible
     Option 1:ESlateVisibility Enum=Collapsed
     Index="Equal Exactly (Text)"!(A:"Text (by ref)"=@n1, B:"Text (by ref)"=@n2) -> (ReturnValue:bool)
    ) -> (ReturnValue:ESlateVisibility Enum)
   )
   > SetBrushFromTexture(
    self=@n3
    Texture=Select!(
     XP:Texture 2D Object Reference=Get XPIcon() -> (XPIcon:Texture 2D Object Reference)
     Bars:Texture 2D Object Reference=Get GoldIcon() -> (GoldIcon:Texture 2D Object Reference)
     Custom:Texture 2D Object Reference
     Index:EFortEarnedSubRewardType Enum=@n1.InType
    ) -> (ReturnValue:Texture 2D Object Reference)
    bMatchSize=false
   )
   > SetVisibility(self=Get Overlay_Reward() -> (Overlay_Reward:Overlay Object Reference), InVisibility=HitTestInvisible)
   > SetText(self=Get Text_RewardQuantity(), InText=Get InText() -> (InText:Text))
  ]}
]
variables [
 GoldIcon {type=Texture2D, default=/Game/UI/Foundation/Textures/Icons/Athena/T-T-Icon-BR-GoldBars-UI-Icon}
 IsBounty {type=boolean}
 StyleType {type=string, default=None}
 Urgent Quest Data {
  type=UrgentQuestData
  default {
   acceptingPlayer=null
   alertIcon=null
   bDisplayAllObjectivesIndividually=true
   bDisplayHUDData=true
   bDisplayMultiStepExpanded=true
   bDisplayStartedEventUI=true
   bDisplayTimer=true
   bHideQuestProgressBar=false
   bIsSharedUrgentQuest=false
   bOverrideWidgetPrimaryColor=false
   bQuestCompleteOnTimeExpiration=false
   bQuickIntro=false
   bQuickOutro=false
   bReverseProgressBar=true
   bRewardPersist=false
   bShowBountyPriceAndDistance=false
   bShowBountyThreatInformation=false
   bountyPriceImage=null
   characterIcon=null
   displayPlayer=null
   eventDescription=Description
   eventDescriptionGamepad=""
   eventDescriptionTouch=""
   eventFailureMessage=""
   eventMainTitle=""
   eventStartTime=0001.01.01-00.00.00
   eventSubtitle=Subtitle
   eventSubtitleSecondary=SubtitleSecondary
   eventTitle=Title
   objectiveCategory=Default
   questProgressTextFormat="{CurrentProgress} / {MaxProgress}"
   socialAvatarBrushPtr=null
   sortPriority=0
   specialEventStartedSound=null
   totalEventTime=60
   alertIconBorderColor=<rgba 1 0.4564 0 1>
   alertIconBrushSize=<0 0>
   eventDescriptionOverrides []
   eventTag {tagName=None}
   eventTimerHandle {}
   overriddenWidgetPrimaryColor=<rgba 1 0.4564 0 1>
   visualStyleTag {gameplayTags []}
  }
 }
 XPIcon {type=Texture2D, default=/Game/UI/Foundation/Textures/Icons/Items/UniversalXP/T_UI_UniversalXP_Icon_Standard_64}
]
widgets {
 info {inheritedWidgetCount=3, parentClass=/Script/FortniteUI.AthenaSpecialEventEntryIcon, rootWidgetClass=&Overlay, widgetCount=14}
 widgets [
  Overlay_Root {widgetClassPath=&Overlay, widget=Overlay_Root}
  SizeBox_0 {
   widgetClassPath=&SizeBox
   parent=Overlay_Root
   widget=SizeBox_0
   properties {bOverride_MinDesiredHeight=true, minDesiredHeight=60}
   slot {class=OverlaySlot, properties {horizontalAlignment=HAlign_Center, verticalAlignment=VAlign_Center}}
  }
  Overlay_Reward {widgetClassPath=&Overlay, bIsVariable=true, parent=SizeBox_0, widget=Overlay_Reward, slot {class=SizeBoxSlot}}
  Image_RewardBG {
   widgetClassPath=&Image
   bIsVariable=true
   parent=Overlay_Reward
   widget=Image_RewardBG
   properties {
    visibility=SelfHitTestInvisible
    brush {
     bIsDynamicallyLoaded=false
     drawAs=Box
     imageType=NoImage
     mirroring=NoMirror
     resourceName=None
     resourceObject=/Game/Athena/HUD/Discoverability/materials/MI_UI_Objectives_RewardIcon
     tiling=NoTile
     imageSize=<32 32>
     margin=<ltrb 0.5 0.5 0.5 0.5>
     outlineSettings {
      bUseBrushTransparency=false
      roundingType=HalfHeightRadius
      width=0
      color {colorUseRule=UseColor_Specified, specifiedColor=<rgba 0 0 0 0>}
      cornerRadii=<0 0 0 0>
     }
     tintColor {colorUseRule=UseColor_Specified, specifiedColor=<rgba 1 1 1 0.7391>}
     uVRegion {bIsValid=false, max=<0 0>, min=<0 0>}
    }
    colorAndOpacity=<rgba 1 1 1 0>
   }
   slot {class=OverlaySlot, properties {horizontalAlignment=HAlign_Fill, verticalAlignment=VAlign_Fill}}
  }
  VerticalBox_0 {
   widgetClassPath=&VerticalBox
   parent=Overlay_Reward
   widget=VerticalBox_0
   properties {bOverride_Cursor=true}
   slot {class=OverlaySlot, properties {horizontalAlignment=HAlign_Fill, verticalAlignment=VAlign_Center}}
  }
  Overlay_0 {widgetClassPath=&Overlay, parent=VerticalBox_0, widget=Overlay_0, slot {class=VerticalBoxSlot}}
  Image_UrgentIcon {
   widgetClassPath=&Image
   bInherited=true
   bIsVariable=true
   parent=Overlay_0
   widget=Image_UrgentIcon
   properties {
    visibility=Hidden
    brush {
     bIsDynamicallyLoaded=false
     drawAs=Image
     imageType=NoImage
     mirroring=NoMirror
     resourceName=None
     resourceObject=/Game/Athena/HUD/Discoverability/Icons/T_UI_UrgentQuest_Exploration
     tiling=NoTile
     imageSize=<50 50>
     margin=<ltrb 0 0 0 0>
     outlineSettings {
      bUseBrushTransparency=false
      roundingType=HalfHeightRadius
      width=0
      color {colorUseRule=UseColor_Specified, specifiedColor=<rgba 0 0 0 0>}
      cornerRadii=<0 0 0 0>
     }
     tintColor {colorUseRule=UseColor_Specified, specifiedColor=<rgba 1 1 1 1>}
     uVRegion {bIsValid=false, max=<0 0>, min=<0 0>}
    }
   }
   slot {class=OverlaySlot, properties {horizontalAlignment=HAlign_Center, verticalAlignment=VAlign_Bottom}}
  }
  Text_UrgentIcon {
   widgetClassPath=&CommonTextBlock
   bInherited=true
   bIsVariable=true
   parent=Overlay_0
   widget=Text_UrgentIcon
   properties {
    bIsEnabled=false
    justification=Center
    text="!"
    visibility=Collapsed
    font {
     bForceMonospaced=false
     bMaterialIsStencil=false
     fontMaterial=null
     fontObject=&Burbank
     letterSpacing=0
     monospacedWidth=1
     size=20
     skewAmount=0
     typefaceFontName=BurbankRegular-Black
     outlineSettings {
      bApplyOutlineToDropShadows=false
      bMiteredCorners=false
      bSeparateFillAlpha=false
      outlineMaterial=null
      outlineSize=0
      outlineColor=<rgba 0 0 0 1>
     }
    }
    renderTransform {angle=0, scale=<1 1>, shear=<-8 0>, translation=<0 0>}
    shadowOffset=<0 0>
   }
   slot {class=OverlaySlot, properties {horizontalAlignment=HAlign_Center, verticalAlignment=VAlign_Bottom}}
  }
  Text_RewardQuantity {
   widgetClassPath=/Script/FortniteUI.FortRichTextBlockKeybind
   bIsVariable=true
   parent=VerticalBox_0
   widget=Text_RewardQuantity
   properties {
    bOverrideDefaultStyle=true
    justification=Center
    keybindWidgetClass=/FortUILibrary/Input/KeybindWidgetDark_Small.KeybindWidgetDark_Small_C
    minDesiredWidth=64
    mobileTextBlockScale=1.2
    scrollStyle=/Game/Athena/Prototype/Blueprints/Galileo/UI/TextScrollStyle-Base-Ticker.TextScrollStyle-Base-Ticker_C
    text="<img id=\"BarsIcon\" width=\"22\"></>000"
    textTransformPolicy=ToUpper
    decoratorClasses [/Game/Athena/HUD/Discoverability/UrgentQuestRichTextDecorators.UrgentQuestRichTextDecorators_C]
    defaultTextStyleOverride {
     overflowPolicy=Clip
     transformPolicy=None
     colorAndOpacity {colorUseRule=UseColor_Specified, specifiedColor=<rgba 1 1 1 1>}
     font {
      bForceMonospaced=false
      bMaterialIsStencil=false
      fontMaterial=null
      fontObject=&BurbankSmall
      letterSpacing=0
      monospacedWidth=1
      size=12
      skewAmount=0
      typefaceFontName=Bold
      outlineSettings {
       bApplyOutlineToDropShadows=false
       bMiteredCorners=false
       bSeparateFillAlpha=false
       outlineMaterial=null
       outlineSize=0
       outlineColor=<rgba 0 0 0 0.502>
      }
     }
     highlightColor {colorUseRule=UseColor_Specified, specifiedColor=<rgba 0 0 0 0>}
     highlightShape {
      bIsDynamicallyLoaded=false
      drawAs=Image
      imageType=NoImage
      mirroring=NoMirror
      resourceName=None
      resourceObject=null
      tiling=NoTile
      imageSize=<32 32>
      margin=<ltrb 0 0 0 0>
      outlineSettings {
       bUseBrushTransparency=false
       roundingType=HalfHeightRadius
       width=0
       color {colorUseRule=UseColor_Specified, specifiedColor=<rgba 0 0 0 0>}
       cornerRadii=<0 0 0 0>
      }
      tintColor {colorUseRule=UseColor_Specified, specifiedColor=<rgba 1 1 1 1>}
      uVRegion {bIsValid=false, max=<0 0>, min=<0 0>}
     }
     selectedBackgroundColor {colorUseRule=UseColor_Foreground, specifiedColor=<rgba 1 0 1 1>}
     shadowColorAndOpacity=<rgba 0 0 0 0>
     shadowOffset=<0 0>
     strikeBrush {
      bIsDynamicallyLoaded=false
      drawAs=Image
      imageType=NoImage
      mirroring=NoMirror
      resourceName=None
      resourceObject=null
      tiling=NoTile
      imageSize=<32 32>
      margin=<ltrb 0 0 0 0>
      outlineSettings {
       bUseBrushTransparency=false
       roundingType=HalfHeightRadius
       width=0
       color {colorUseRule=UseColor_Specified, specifiedColor=<rgba 0 0 0 0>}
       cornerRadii=<0 0 0 0>
      }
      tintColor {colorUseRule=UseColor_Specified, specifiedColor=<rgba 1 1 1 1>}
      uVRegion {bIsValid=false, max=<0 0>, min=<0 0>}
     }
     underlineBrush {
      bIsDynamicallyLoaded=false
      drawAs=Image
      imageType=NoImage
      mirroring=NoMirror
      resourceName=None
      resourceObject=null
      tiling=NoTile
      imageSize=<32 32>
      margin=<ltrb 0 0 0 0>
      outlineSettings {
       bUseBrushTransparency=false
       roundingType=HalfHeightRadius
       width=0
       color {colorUseRule=UseColor_Specified, specifiedColor=<rgba 0 0 0 0>}
       cornerRadii=<0 0 0 0>
      }
      tintColor {colorUseRule=UseColor_Specified, specifiedColor=<rgba 1 1 1 1>}
      uVRegion {bIsValid=false, max=<0 0>, min=<0 0>}
     }
    }
    desiredKeybindWidgetSize=<16 16>
   }
   slot {class=VerticalBoxSlot, properties {horizontalAlignment=HAlign_Center, verticalAlignment=VAlign_Center}}
  }
  Overlay_CheckMark {
   widgetClassPath=&Overlay
   parent=Overlay_Root
   widget=Overlay_CheckMark
   properties {renderTransform {angle=0, scale=<1.5 1.5>, shear=<0 0>, translation=<2 -6>}}
   slot {class=OverlaySlot, properties {horizontalAlignment=HAlign_Right, verticalAlignment=VAlign_Center, padding=<ltrb 0 -4 0 -4>}}
  }
  Image_CheckmarkGlow {
   widgetClassPath=&Image
   bIsVariable=true
   parent=Overlay_CheckMark
   widget=Image_CheckmarkGlow
   properties {
    brush {
     bIsDynamicallyLoaded=false
     drawAs=Image
     imageType=NoImage
     mirroring=NoMirror
     resourceName=None
     resourceObject=/Game/Athena/HUD/Materials/M_CountdownBG_SolidGlow
     tiling=NoTile
     imageSize=<32 32>
     margin=<ltrb 0 0 0 0>
     outlineSettings {
      bUseBrushTransparency=false
      roundingType=HalfHeightRadius
      width=0
      color {colorUseRule=UseColor_Specified, specifiedColor=<rgba 0 0 0 0>}
      cornerRadii=<0 0 0 0>
     }
     tintColor {colorUseRule=UseColor_Specified, specifiedColor=<rgba 1 1 1 1>}
     uVRegion {bIsValid=false, max=<0 0>, min=<0 0>}
    }
    colorAndOpacity=<rgba 0 0.0958 1 0>
    renderTransform {angle=7, scale=<1.9 1>, shear=<0 0>, translation=<0 0>}
   }
   slot {class=OverlaySlot, properties {horizontalAlignment=HAlign_Fill, verticalAlignment=VAlign_Fill}}
  }
  Image_Spark {
   widgetClassPath=&Image
   bIsVariable=true
   parent=Overlay_CheckMark
   widget=Image_Spark
   properties {
    renderOpacity=0
    brush {
     bIsDynamicallyLoaded=false
     drawAs=Image
     imageType=NoImage
     mirroring=NoMirror
     resourceName=None
     resourceObject=/Game/UI/InGame/Quest/Materials/M_UI_SparkStar
     tiling=NoTile
     imageSize=<16 16>
     margin=<ltrb 0 0 0 0>
     outlineSettings {
      bUseBrushTransparency=false
      roundingType=HalfHeightRadius
      width=0
      color {colorUseRule=UseColor_Specified, specifiedColor=<rgba 0 0 0 0>}
      cornerRadii=<0 0 0 0>
     }
     tintColor {colorUseRule=UseColor_Specified, specifiedColor=<rgba 1 1 1 1>}
     uVRegion {bIsValid=false, max=<0 0>, min=<0 0>}
    }
    renderTransform {angle=0, scale=<2.5 2.5>, shear=<0 0>, translation=<0 4>}
   }
   slot {class=OverlaySlot, properties {horizontalAlignment=HAlign_Fill, verticalAlignment=VAlign_Fill}}
  }
  Image_CheckMark {
   widgetClassPath=&Image
   parent=Overlay_CheckMark
   widget=Image_CheckMark
   properties {
    renderOpacity=0
    brush {
     bIsDynamicallyLoaded=false
     drawAs=Image
     imageType=FullColor
     mirroring=NoMirror
     resourceName=None
     resourceObject=/Game/UI/Foundation/Textures/Icons/Checks/T_UI_Icon_Check_MP_32
     tiling=NoTile
     imageSize=<12 12>
     margin=<ltrb 0 0 0 0>
     outlineSettings {
      bUseBrushTransparency=false
      roundingType=HalfHeightRadius
      width=0
      color {colorUseRule=UseColor_Specified, specifiedColor=<rgba 0 0 0 0>}
      cornerRadii=<0 0 0 0>
     }
     tintColor {colorUseRule=UseColor_Specified, specifiedColor=<rgba 0 1 0.1441 1>}
     uVRegion {bIsValid=false, max=<0 0>, min=<0 0>}
    }
   }
   slot {class=OverlaySlot, properties {horizontalAlignment=HAlign_Center, verticalAlignment=VAlign_Center}}
  }
  Image_AlertBackground {
   widgetClassPath=&Image
   bInherited=true
   bIsVariable=true
   parent=Overlay_Root
   widget=Image_AlertBackground
   properties {
    visibility=Collapsed
    brush {
     bIsDynamicallyLoaded=false
     drawAs=RoundedBox
     imageType=NoImage
     mirroring=NoMirror
     resourceName=None
     resourceObject=null
     tiling=NoTile
     imageSize=<10 10>
     margin=<ltrb 0.5 0.5 0.5 0.5>
     outlineSettings {
      bUseBrushTransparency=false
      roundingType=HalfHeightRadius
      width=0
      color {colorUseRule=UseColor_Specified, specifiedColor=<rgba 0 0 0 0>}
      cornerRadii=<0 0 0 0>
     }
     tintColor {colorUseRule=UseColor_Specified, specifiedColor=<rgba 1 1 1 0>}
     uVRegion {bIsValid=false, max=<0 0>, min=<0 0>}
    }
   }
   slot {class=OverlaySlot, properties {horizontalAlignment=HAlign_Center, verticalAlignment=VAlign_Center}}
  }
 ]
}
