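# Match objectives HUD widget that displays an objective's progress text, with animated feedback for progress gains and completion.
# Construct hides the main overlay until SetObjectiveProgressText stores the text and shows it.
# Exposed functions play a glow pulse or a completion animation that greens the text and reveals checkmark, spark and glow; reset replays completion in Reverse using its end time.
# A boolean toggles the checkmark overlay, also applied in designer preview.
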WidgetBlueprint /Game/Athena/HUD/Discoverability/MatchObjectivesWidgets/WBP_Athena_Objectives_Progress
animations [
 Anim_ProgressGained {
  binding_count=1
  length_seconds=0.5
  object=&self:Anim_ProgressGained
  sequence {bindings [Image_ProgressGlow {
     id=#1
     possessed_class=Image
     tracks [
      {
       type=MovieSceneColorTrack
       display_name=Color and Opacity
       sections [{
         type=MovieSceneColorSection
         end_frame=7
         start_frame=0
         channels [
          Color.R {type=MovieSceneScriptingFloatChannel}
          Color.G {type=MovieSceneScriptingFloatChannel}
          Color.B {type=MovieSceneScriptingFloatChannel}
          Color.A {type=MovieSceneScriptingFloatChannel, keys [[0,0],[4,1],[7,0]]}
         ]
        }]
      }
      {
       type=MovieScene2DTransformTrack
       sections [{
         type=MovieScene2DTransformSection
         end_frame=10
         start_frame=0
         channels [
          Translation.X {type=MovieSceneScriptingFloatChannel}
          Translation.Y {type=MovieSceneScriptingFloatChannel}
          Angle {type=MovieSceneScriptingFloatChannel, keys [[0,-2],[4,0],[10,7]]}
          Scale.X {type=MovieSceneScriptingFloatChannel, keys [[0,1],[4,2.4033],[7,1]]}
          Scale.Y {type=MovieSceneScriptingFloatChannel, keys [[0,1],[4,1.1365],[7,1]]}
          Shear.X {type=MovieSceneScriptingFloatChannel, const [0,1]}
          Shear.Y {type=MovieSceneScriptingFloatChannel, const [0,1]}
         ]
        }]
      }
     ]
    }]}
 }
 Anim_ObjectiveCompleted {
  binding_count=4
  length_seconds=0.85
  object=&self:Anim_ObjectiveCompleted
  sequence {
   bindings [
    Image_CheckMark {
     id=#2
     possessed_class=Image
     tracks [
      {
       type=MovieSceneFloatTrack
       display_name=Render Opacity
       sections [{
         type=MovieSceneFloatSection
         end_frame=51
         start_frame=0
         channels [None {type=MovieSceneScriptingFloatChannel, thinned_from=5, keys [[0,0],[22.444,0],[36,1],[51,1]]}]
        }]
      }
      {
       type=MovieScene2DTransformTrack
       sections [{
         type=MovieScene2DTransformSection
         end_frame=51
         start_frame=0
         channels [
          Translation.X {type=MovieSceneScriptingFloatChannel}
          Translation.Y {type=MovieSceneScriptingFloatChannel}
          Angle {type=MovieSceneScriptingFloatChannel}
          Scale.X {type=MovieSceneScriptingFloatChannel, keys [[19.054,3],[36,1],[51,1]]}
          Scale.Y {type=MovieSceneScriptingFloatChannel, keys [[19.054,3],[36,1],[51,1]]}
          Shear.X {type=MovieSceneScriptingFloatChannel, const [0,1]}
          Shear.Y {type=MovieSceneScriptingFloatChannel, const [0,1]}
         ]
        }]
      }
     ]
    }
    Image_Spark {
     id=#3
     possessed_class=Image
     tracks [
      {
       type=MovieSceneFloatTrack
       display_name=Render Opacity
       sections [{
         type=MovieSceneFloatSection
         end_frame=50
         start_frame=0
         channels [None {type=MovieSceneScriptingFloatChannel, keys [[0,0],[30.547,0],[37.364,1],[50.999,0]]}]
        }]
      }
      {
       type=MovieScene2DTransformTrack
       sections [{
         type=MovieScene2DTransformSection
         end_frame=51
         start_frame=30
         channels [
          Translation.X {type=MovieSceneScriptingFloatChannel}
          Translation.Y {type=MovieSceneScriptingFloatChannel}
          Angle {type=MovieSceneScriptingFloatChannel}
          Scale.X {type=MovieSceneScriptingFloatChannel, keys [[30.547,2.5],[50.999,1]]}
          Scale.Y {type=MovieSceneScriptingFloatChannel, keys [[30.547,2.5],[50.999,1]]}
          Shear.X {type=MovieSceneScriptingFloatChannel}
          Shear.Y {type=MovieSceneScriptingFloatChannel}
         ]
        }]
      }
     ]
    }
    Text_ProgressValue {
     id=#4
     possessed_class=CommonTextBlock
     tracks [
      {
       type=MovieSceneColorTrack
       display_name=Color and Opacity
       sections [{
         type=MovieSceneColorSection
         end_frame=50
         start_frame=0
         channels [
          Color.R {type=MovieSceneScriptingFloatChannel, keys [[0,1],[30,1],[38,0],[50,0]]}
          Color.G {type=MovieSceneScriptingFloatChannel, const [1,3]}
          Color.B {type=MovieSceneScriptingFloatChannel, keys [[0,1],[30,1],[38,0.1441],[50,0.1441]]}
          Color.A {type=MovieSceneScriptingFloatChannel, const [1,3]}
         ]
        }]
      }
      {
       type=MovieScene2DTransformTrack
       sections [{
         type=MovieScene2DTransformSection
         end_frame=0
         start_frame=0
         channels [
          Translation.X {type=MovieSceneScriptingFloatChannel, const [0,1]}
          Translation.Y {type=MovieSceneScriptingFloatChannel, const [0,1]}
          Angle {type=MovieSceneScriptingFloatChannel, const [0,1]}
          Scale.X {type=MovieSceneScriptingFloatChannel, const [1,1]}
          Scale.Y {type=MovieSceneScriptingFloatChannel, const [1,1]}
          Shear.X {type=MovieSceneScriptingFloatChannel, const [0,1]}
          Shear.Y {type=MovieSceneScriptingFloatChannel, const [0,1]}
         ]
        }]
      }
     ]
    }
    Image_CheckmarkGlow {
     id=#5
     possessed_class=Image
     tracks [
      {
       type=MovieSceneColorTrack
       display_name=Color and Opacity
       sections [{
         type=MovieSceneColorSection
         end_frame=51
         start_frame=0
         channels [
          Color.R {type=MovieSceneScriptingFloatChannel, const [0,2]}
          Color.G {type=MovieSceneScriptingFloatChannel, const [1,2]}
          Color.B {type=MovieSceneScriptingFloatChannel, const [0.1441,3]}
          Color.A {type=MovieSceneScriptingFloatChannel, const [1,2]}
         ]
        }]
      }
      {
       type=MovieSceneFloatTrack
       display_name=Render Opacity
       sections [{
         type=MovieSceneFloatSection
         end_frame=51
         start_frame=0
         channels [None {type=MovieSceneScriptingFloatChannel, keys [[0,0],[30.547,0],[37.364,1],[50.999,0]]}]
        }]
      }
     ]
    }
   ]
   display_rate {numerator=60}
  }
 }
]
graphs [
 EventGraph {nodes [
   Event Construct()
   > n1 = SetVisibility(self=Get Overlay_Main() -> (Overlay_Main:Overlay Object Reference), InVisibility=Hidden)
   Update Checkmark Visibility!(execute=[@n3.then, @n1.then], self, IncludeCheckmarkInSuccess:bool=Get Include Checkmark in Succes() -> (Include Checkmark in Succes:bool)) -> (then)
   n2 = Event PreConstruct()
   > n3 = Branch(Condition=@n2.IsDesignTime)
  ]}
 GetProgressText {nodes [
   GetProgressText() {is=title K2Node_FunctionEntry}
   > Return Node!(execute, currentProgress:Text=Get CurrentProgressText() -> (CurrentProgressText:Text))
  ]}
 PlayObjectiveCompletedAnim {nodes [
   PlayObjectiveCompletedAnim() {is=title K2Node_FunctionEntry}
   > PlayAnimation(InAnimation=Get Anim_ObjectiveCompleted() -> (Anim_ObjectiveCompleted:Widget Animation Object Reference))
  ]}
 PlayProgressGainedAnim {nodes [
   PlayProgressGainedAnim() {is=title K2Node_FunctionEntry}
   > PlayAnimation(InAnimation=Get Anim_ProgressGained() -> (Anim_ProgressGained:Widget Animation Object Reference))
  ]}
 ResetCompletedAnim {nodes [
   ResetCompletedAnim() {is=title K2Node_FunctionEntry}
   n1 = Get Anim_ObjectiveCompleted()
   > PlayAnimation(InAnimation=@n1, StartAtTime=GetEndTime(self=@n1), PlayMode=Reverse)
  ]}
 SetObjectiveProgressText {nodes [
   n1 = SetObjectiveProgressText!() -> (then, Text:Text) {is=title K2Node_FunctionEntry}
   > n2 = Set CurrentProgressText(CurrentProgressText:Text=@n1) -> (Output_Get:Text)
   > "SetText (Text)"(self=Get Text_ProgressValue() -> (Text_ProgressValue:Common Text Object Reference), InText=@n2)
   > SetVisibility(self=Get Overlay_Main(), InVisibility=SelfHitTestInvisible)
  ]}
 Update Checkmark Visibility {nodes [
   n1 = Update Checkmark Visibility!() -> (then, IncludeCheckmarkInSuccess:bool=~) {is=title K2Node_FunctionEntry}
   > SetVisibility(
    self=Get Overlay_CheckMark() -> (Overlay_CheckMark:Overlay Object Reference)
    InVisibility=Select(Option 0=Collapsed, Option 1=SelfHitTestInvisible, Index=@n1)
   )
  ]}
]
variables [
 CurrentProgressText {type=string, default=""}
 Include Checkmark in Succes {type=boolean, default=true}
]
widgets {
 info {inheritedWidgetCount=0, parentClass=&UserWidget, rootWidgetClass=&Overlay, widgetCount=8}
 widgets [
  Overlay_Root {widgetClassPath=&Overlay, widget=Overlay_Root}
  Overlay_Main {
   widgetClassPath=&Overlay
   bIsVariable=true
   parent=Overlay_Root
   widget=Overlay_Main
   slot {class=OverlaySlot, properties {horizontalAlignment=HAlign_Right, verticalAlignment=VAlign_Center}}
  }
  Image_ProgressGlow {
   widgetClassPath=&Image
   bIsVariable=true
   parent=Overlay_Main
   widget=Image_ProgressGlow
   properties {
    brush {
     bIsDynamicallyLoaded=false
     drawAs=Image
     imageType=NoImage
     mirroring=NoMirror
     resourceName=None
     resourceObject=/Game/Athena/HUD/Materials/M_CountdownBG_SolidGlow
     tiling=NoTile
     imageSize=<32 32>
     margin=<ltrb 0 0 0 0>
     outlineSettings {
      bUseBrushTransparency=false
      roundingType=HalfHeightRadius
      width=0
      color {colorUseRule=UseColor_Specified, specifiedColor=<rgba 0 0 0 0>}
      cornerRadii=<0 0 0 0>
     }
     tintColor {colorUseRule=UseColor_Specified, specifiedColor=<rgba 1 1 1 1>}
     uVRegion {bIsValid=false, max=<0 0>, min=<0 0>}
    }
    colorAndOpacity=<rgba 0 0.0958 1 0>
    renderTransform {angle=7, scale=<1.9 1>, shear=<0 0>, translation=<0 0>}
   }
   slot {class=OverlaySlot, properties {horizontalAlignment=HAlign_Fill, verticalAlignment=VAlign_Fill, padding=<ltrb 0 -4 0 -4>}}
  }
  Text_ProgressValue {
   widgetClassPath=&CommonTextBlock
   bIsVariable=true
   parent=Overlay_Main
   widget=Text_ProgressValue
   properties {
    mobileFontSizeMultiplier=1.2
    text=0/1
    textTransformPolicy=ToUpper
    font {
     bForceMonospaced=false
     bMaterialIsStencil=false
     fontMaterial=null
     fontObject=&HeadingNow
     letterSpacing=70
     monospacedWidth=1
     size=12
     skewAmount=0
     typefaceFontName=75 Medium
     outlineSettings {
      bApplyOutlineToDropShadows=false
      bMiteredCorners=false
      bSeparateFillAlpha=false
      outlineMaterial=null
      outlineSize=2
      outlineColor=<rgba 0 0.013 0.117 1>
     }
    }
    shadowOffset=<0 0>
   }
   slot {class=OverlaySlot, properties {horizontalAlignment=HAlign_Right, verticalAlignment=VAlign_Center}}
  }
  Overlay_CheckMark {
   widgetClassPath=&Overlay
   bIsVariable=true
   parent=Overlay_Root
   widget=Overlay_CheckMark
   properties {renderTransform {angle=0, scale=<1 1>, shear=<0 0>, translation=<28 -2>}}
   slot {class=OverlaySlot, properties {horizontalAlignment=HAlign_Right, verticalAlignment=VAlign_Center, padding=<ltrb 0 -4 0 -4>}}
  }
  Image_CheckmarkGlow {
   widgetClassPath=&Image
   bIsVariable=true
   parent=Overlay_CheckMark
   widget=Image_CheckmarkGlow
   properties {
    brush {
     bIsDynamicallyLoaded=false
     drawAs=Image
     imageType=NoImage
     mirroring=NoMirror
     resourceName=None
     resourceObject=&M_CountdownBG_SolidGlow
     tiling=NoTile
     imageSize=<32 32>
     margin=<ltrb 0 0 0 0>
     outlineSettings {
      bUseBrushTransparency=false
      roundingType=HalfHeightRadius
      width=0
      color {colorUseRule=UseColor_Specified, specifiedColor=<rgba 0 0 0 0>}
      cornerRadii=<0 0 0 0>
     }
     tintColor {colorUseRule=UseColor_Specified, specifiedColor=<rgba 1 1 1 1>}
     uVRegion {bIsValid=false, max=<0 0>, min=<0 0>}
    }
    colorAndOpacity=<rgba 0 0.0958 1 0>
    renderTransform {angle=7, scale=<1.9 1>, shear=<0 0>, translation=<0 0>}
   }
   slot {class=OverlaySlot, properties {horizontalAlignment=HAlign_Fill, verticalAlignment=VAlign_Fill}}
  }
  Image_Spark {
   widgetClassPath=&Image
   bIsVariable=true
   parent=Overlay_CheckMark
   widget=Image_Spark
   properties {
    renderOpacity=0
    brush {
     bIsDynamicallyLoaded=false
     drawAs=Image
     imageType=NoImage
     mirroring=NoMirror
     resourceName=None
     resourceObject=/Game/UI/InGame/Quest/Materials/M_UI_SparkStar
     tiling=NoTile
     imageSize=<16 16>
     margin=<ltrb 0 0 0 0>
     outlineSettings {
      bUseBrushTransparency=false
      roundingType=HalfHeightRadius
      width=0
      color {colorUseRule=UseColor_Specified, specifiedColor=<rgba 0 0 0 0>}
      cornerRadii=<0 0 0 0>
     }
     tintColor {colorUseRule=UseColor_Specified, specifiedColor=<rgba 1 1 1 1>}
     uVRegion {bIsValid=false, max=<0 0>, min=<0 0>}
    }
    renderTransform {angle=0, scale=<2.5 2.5>, shear=<0 0>, translation=<0 4>}
   }
   slot {class=OverlaySlot, properties {horizontalAlignment=HAlign_Fill, verticalAlignment=VAlign_Fill}}
  }
  Image_CheckMark {
   widgetClassPath=&Image
   parent=Overlay_CheckMark
   widget=Image_CheckMark
   properties {
    renderOpacity=0
    brush {
     bIsDynamicallyLoaded=false
     drawAs=Image
     imageType=FullColor
     mirroring=NoMirror
     resourceName=None
     resourceObject=/Game/UI/Foundation/Textures/Icons/Checks/T_UI_Icon_Check_MP_32
     tiling=NoTile
     imageSize=<16 16>
     margin=<ltrb 0 0 0 0>
     outlineSettings {
      bUseBrushTransparency=false
      roundingType=HalfHeightRadius
      width=0
      color {colorUseRule=UseColor_Specified, specifiedColor=<rgba 0 0 0 0>}
      cornerRadii=<0 0 0 0>
     }
     tintColor {colorUseRule=UseColor_Specified, specifiedColor=<rgba 0 1 0.1441 1>}
     uVRegion {bIsValid=false, max=<0 0>, min=<0 0>}
    }
    renderTransform {angle=0, scale=<1 1>, shear=<0 0>, translation=<0 2>}
   }
   slot {class=OverlaySlot, properties {horizontalAlignment=HAlign_Center, verticalAlignment=VAlign_Center}}
  }
 ]
}
