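# Optional-objective row for the Athena match objectives HUD, built on AthenaObjectiveTaskEntryBase, showing index, description, progress and an XP or Bars reward.
# Base-class events feed reward text and the nested progress widget, then broadcast progress and completion through dispatchers.
# SetActiveVisuals switches three enum states, with active changes skipped while DelayInactiveVisualsTimer runs.
# The Completed animation drives both switchers, dims the row and pops a checkmark with glow and spark.
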WidgetBlueprint /Game/Athena/HUD/Discoverability/MatchObjectivesWidgets/WBP_Athena_Objectives_OptionalEntry
animations [
 SetActiveAnimation {
  binding_count=3
  length_seconds=0.3
  object=&self:SetActiveAnimation
  sequence {bindings [
    CommonVisibilitySwitcher_2 {
     id=#1
     possessed_class=CommonVisibilitySwitcher
     tracks [{
       type=MovieSceneIntegerTrack
       display_name=Active Widget Index
       sections [{type=MovieSceneIntegerSection, end_frame=0, start_frame=0, channels [None {type=MovieSceneScriptingIntegerChannel, const [0,1]}]}]
      }]
    }
    Switcher_Progress {
     id=#2
     possessed_class=CommonActivatableWidgetSwitcher
     tracks [{
       type=MovieSceneIntegerTrack
       display_name=Active Widget Index
       sections [{type=MovieSceneIntegerSection, end_frame=0, start_frame=0, channels [None {type=MovieSceneScriptingIntegerChannel, const [0,1]}]}]
      }]
    }
    Border_Disable {
     id=#3
     possessed_class=Border
     tracks [{
       type=MovieSceneColorTrack
       display_name=Content Color and Opacity
       sections [{
         type=MovieSceneColorSection
         end_frame=6
         start_frame=0
         channels [
          Color.R {type=MovieSceneScriptingFloatChannel, const [1,2]}
          Color.G {type=MovieSceneScriptingFloatChannel, const [1,2]}
          Color.B {type=MovieSceneScriptingFloatChannel, const [1,2]}
          Color.A {type=MovieSceneScriptingFloatChannel, const [1,2]}
         ]
        }]
      }]
    }
   ]}
 }
 Completed {
  binding_count=6
  length_seconds=0.45
  object=&self:Completed
  sequence {bindings [
    Image_Check {
     id=#4
     possessed_class=Image
     tracks [
      {
       type=MovieSceneFloatTrack
       display_name=Render Opacity
       sections [{type=MovieSceneFloatSection, channels [None {type=MovieSceneScriptingFloatChannel, keys [[0,0],[3.9473,1]]}]}]
      }
      {
       type=MovieScene2DTransformTrack
       sections [{
         type=MovieScene2DTransformSection
         channels [
          Translation.X {type=MovieSceneScriptingFloatChannel}
          Angle {type=MovieSceneScriptingFloatChannel}
          Scale.X {type=MovieSceneScriptingFloatChannel, keys [[0,1.5],[2.9607,0.45],[3.9473,0.6]]}
          Scale.Y {type=MovieSceneScriptingFloatChannel, keys [[0,1.5],[2.9607,0.45],[3.9473,0.6]]}
          Shear.X {type=MovieSceneScriptingFloatChannel}
          Shear.Y {type=MovieSceneScriptingFloatChannel}
         ]
        }]
      }
     ]
    }
    Switcher_Progress {
     id=#5
     possessed_class=CommonActivatableWidgetSwitcher
     tracks [{
       type=MovieSceneIntegerTrack
       display_name=Active Widget Index
       sections [{type=MovieSceneIntegerSection, end_frame=6, start_frame=0, channels [None {type=MovieSceneScriptingIntegerChannel, const [1,2]}]}]
      }]
    }
    Border_Disable {
     id=#6
     possessed_class=Border
     tracks [{
       type=MovieSceneColorTrack
       display_name=Content Color and Opacity
       sections [{
         type=MovieSceneColorSection
         end_frame=9
         start_frame=0
         channels [
          Color.R {type=MovieSceneScriptingFloatChannel, keys [[0,1],[7,1],[9,0.3]]}
          Color.G {type=MovieSceneScriptingFloatChannel, keys [[0,1],[7,1],[9,0.3]]}
          Color.B {type=MovieSceneScriptingFloatChannel, keys [[0,1],[7,1],[9,0.3]]}
          Color.A {type=MovieSceneScriptingFloatChannel, keys [[0,1],[7,1],[9,0.7]]}
         ]
        }]
      }]
    }
    Image_Spark {
     id=#7
     possessed_class=Image
     tracks [
      {
       type=MovieSceneFloatTrack
       display_name=Render Opacity
       sections [{type=MovieSceneFloatSection, channels [None {type=MovieSceneScriptingFloatChannel, keys [[3.1073,0],[4.4047,1],[7,0]]}]}]
      }
      {
       type=MovieScene2DTransformTrack
       sections [{
         type=MovieScene2DTransformSection
         channels [
          Translation.X {type=MovieSceneScriptingFloatChannel}
          Angle {type=MovieSceneScriptingFloatChannel}
          Scale.X {type=MovieSceneScriptingFloatChannel, keys [[3.1073,0.75],[7,1]]}
          Scale.Y {type=MovieSceneScriptingFloatChannel, keys [[3.1073,0.75],[7,1]]}
          Shear.X {type=MovieSceneScriptingFloatChannel}
          Shear.Y {type=MovieSceneScriptingFloatChannel}
         ]
        }]
      }
     ]
    }
    CommonVisibilitySwitcher_2 {
     id=#8
     possessed_class=CommonVisibilitySwitcher
     tracks [{
       type=MovieSceneIntegerTrack
       display_name=Active Widget Index
       sections [{type=MovieSceneIntegerSection, end_frame=9, start_frame=0, channels [None {type=MovieSceneScriptingIntegerChannel, keys [[0,0],[9,1]]}]}]
      }]
    }
    Image_Glow {
     id=#9
     possessed_class=Image
     tracks [
      {
       type=MovieSceneFloatTrack
       display_name=Render Opacity
       sections [{type=MovieSceneFloatSection, channels [None {type=MovieSceneScriptingFloatChannel, keys [[3.003,0],[4.3353,1],[7,0]]}]}]
      }
      {
       type=MovieScene2DTransformTrack
       sections [{
         type=MovieScene2DTransformSection
         channels [
          Translation.X {type=MovieSceneScriptingFloatChannel}
          Angle {type=MovieSceneScriptingFloatChannel}
          Scale.X {type=MovieSceneScriptingFloatChannel, keys [[3.003,1],[7,1.3]]}
          Scale.Y {type=MovieSceneScriptingFloatChannel, keys [[3.003,1],[7,1.3]]}
          Shear.X {type=MovieSceneScriptingFloatChannel}
          Shear.Y {type=MovieSceneScriptingFloatChannel}
         ]
        }]
      }
     ]
    }
   ]}
 }
]
graphs [
 EventGraph {nodes [
   Event OnNewlyCompleted() {is=title K2Node_Event}
   > PlayObjectiveCompletedAnim(self:WBP Athena Objectives Progress Object Reference=Get WBP_Athena_Objectives_Progress() -> (WBP_Athena_Objectives_Progress:WBP Athena Objectives Progress Object Reference))
   > Set OptionalObjectiveCompleted(OptionalObjectiveCompleted:bool=true) -> (Output_Get:bool)
   > SetActiveVisuals!(execute, self, CurrentState:OptionalEntryEnumState Enum=NewEnumerator2) -> (then)
   > Call On Optional Complete!(execute:Exec, self:WBP Athena Objectives Optional Entry Object Reference) -> (then:Exec) {is=title K2Node_CallDelegate}
   n1 = Event OnObjectiveRewardSet!() -> (OutputDelegate, then, RewardQuantity:int=0, RewardType:EFortEarnedSubRewardType Enum=XP) {is=title K2Node_Event}
   > Set Rewards!(execute, self, Quantity:int=@n1.RewardQuantity, Type:EFortEarnedSubRewardType Enum=@n1.RewardType) -> (then) {is=title K2Node_CallFunction}
   > SetVisibility(self=Get Overlay_Reward() -> (Overlay_Reward:Overlay Object Reference), InVisibility=SelfHitTestInvisible)
   Event Construct()
   > n2 = Cast To WBP_Athena_Objectives_Progress!(execute, Object=Get ProgressWidget() -> (ProgressWidget:User Widget Object Reference)) -> (then, CastFailed, AsWBP Athena Objectives Progress:WBP Athena Objectives Progress Object Reference) {is=title K2Node_DynamicCast}
   > Set WBP_Athena_Objectives_Progress(WBP_Athena_Objectives_Progress:WBP Athena Objectives Progress Object Reference=@n2) -> (Output_Get:WBP Athena Objectives Progress Object Reference)
   n3 = Event OnUpdateOptionalObjectiveProgress!() -> (OutputDelegate, then, Text:"Text (by ref)") {is=title K2Node_Event}
   n4 = Get WBP_Athena_Objectives_Progress()
   > SetObjectiveProgressText!(execute, self:WBP Athena Objectives Progress Object Reference=@n4, Text:Text=@n3.Text) -> (then)
   > Call On Optional Progress!(execute:Exec, self:WBP Athena Objectives Optional Entry Object Reference, ProgressText:Text=@n3.Text) -> (then:Exec) {is=title K2Node_CallDelegate}
   > PlayProgressGainedAnim(self:WBP Athena Objectives Progress Object Reference=@n4)
   n5 = Event OnIsActiveSet!() -> (OutputDelegate, then, bNewIsActive:bool) {is=title K2Node_Event}
   > Set IsObjectiveActive(IsObjectiveActive:bool=@n5.bNewIsActive) -> (Output_Get:bool)
   > n6 = Branch(Condition=Is Timer Active by Handle!(Handle:Timer Handle Structure=Get DelayInactiveVisualsTimer() -> (DelayInactiveVisualsTimer:Timer Handle Structure)) -> (ReturnValue:bool))
   SetActiveVisuals(
    execute=@n6.else
    CurrentState=Select(
     Option 0:OptionalEntryEnumState Enum=NewEnumerator0
     Option 1:OptionalEntryEnumState Enum=NewEnumerator1
     Index=Get IsObjectiveActive() -> (IsObjectiveActive:bool)
    ) -> (ReturnValue:OptionalEntryEnumState Enum)
   )
   Event OnInitialized()
   > SetActiveVisuals(CurrentState=NewEnumerator0)
  ]}
 OnOptionalComplete {nodes [OnOptionalComplete() {is=title K2Node_FunctionEntry}]}
 OnOptionalProgress {nodes [OnOptionalProgress!() -> (then, ProgressText:Text) {is=title K2Node_FunctionEntry}]}
 ResetEntry {nodes [
   ResetEntry() {is=title K2Node_FunctionEntry}
   > Set OptionalObjectiveCompleted(OptionalObjectiveCompleted=false)
   > ResetCompletedAnim(self:WBP Athena Objectives Progress Object Reference=Get WBP_Athena_Objectives_Progress())
  ]}
 Set Rewards {nodes [
   n1 = Set Rewards!() -> (then, Quantity:int, Type:EFortEarnedSubRewardType Enum=XP) {is=title K2Node_FunctionEntry}
   > n2 = Set OptionalObjRewardQuantity(OptionalObjRewardQuantity:int=@n1) -> (Output_Get:int=0)
   > SetText(
    self=Get Text_ObjectiveAdditionalReward() -> (Text_ObjectiveAdditionalReward:Fort Rich Text Block Keybind Object Reference)
    InText=Format Text!(
     Format="LOCTEXT(\"{RewardIcon}{RewardQuantity}\")"
     RewardIcon:Text=Select!(
      XP:Text="LOCTEXT(\"<img id=\\\"XPIcon\\\" width=\\\"22\\\"></> \")"
      Bars:Text="LOCTEXT(\"<img id=\\\"BarsIcon\\\" width=\\\"22\\\"></>\")"
      Custom:Text
      Index:EFortEarnedSubRewardType Enum=@n1.Type
     ) -> (ReturnValue:Text)
     RewardQuantity:int=@n2
    ) -> (Result)
   )
   > Return Node()
  ]}
 SetActiveVisuals {nodes [
   n1 = SetActiveVisuals!() -> (then, CurrentState:OptionalEntryEnumState Enum=NewEnumerator0) {is=title K2Node_FunctionEntry}
   > n2 = Switch on OptionalEntryEnumState!(execute, Selection:OptionalEntryEnumState Enum=@n1) -> (NewEnumerator0:Exec, NewEnumerator1:Exec, NewEnumerator2:Exec) {is=title K2Node_SwitchEnum}
   > SetContentColorAndOpacity!(
    execute
    self:Border Object Reference=Get Border_Disable() -> (Border_Disable:Border Object Reference)
    InContentColorAndOpacity:Linear Color Structure=MakeColor!(R:float=1, G:float=1, B:float=1, A:float=1) -> (ReturnValue:Linear Color Structure)
   ) -> (then)
   SetContentColorAndOpacity(execute=@n2.NewEnumerator1, self=Get Border_Disable(), InContentColorAndOpacity=MakeColor(R=1, G=1, B=1, A=1))
   > PlayAnimationForward(InAnimation=Get SetActiveAnimation() -> (SetActiveAnimation:Widget Animation Object Reference))
   PlayAnimationForward(execute=@n2.NewEnumerator2, InAnimation=Get Completed() -> (Completed:Widget Animation Object Reference))
  ]}
 SetIndexEntry {nodes [
   n1 = SetIndexEntry!() -> (then, TopIndex:int, Count:int) {is=title K2Node_FunctionEntry}
   > SetText(
    self=Get Text_ObjectIndex() -> (Text_ObjectIndex:Fort Rich Text Block Keybind Object Reference)
    InText=Format Text!(Format="LOCTEXT(\"<TopCount>{TopIndex}.</>\")", TopIndex:int=@n1) -> (Result)
   )
  ]}
]
variables [
 BarCurrencyItemName {type=string, default=WorldItem:Athena_wadsitemdata}
 Current Objective Index {type=integer, default=2}
 DefaultTint {type=SlateColor, default {colorUseRule=UseColor_Specified, specifiedColor=<rgba 1 0.4564 0 1>}}
 DelayInactiveVisualsTimer {type=TimerHandle}
 IsObjectiveActive {type=boolean}
 OnOptionalComplete {}
 OnOptionalProgress {}
 OptionalObjectiveCompleted {type=boolean}
 OptionalObjRewardQuantity {type=integer}
 StrikeThroughMaterial {type=Object, default=/FortUILibrary/Foundation/Art_ToolKit/AngledBoxes/M_UI_StrikeThrough_Line}
 WBP_Athena_Objectives_Progress {type=WBP_Athena_Objectives_Progress_C, default=null}
]
widgets {
 info {inheritedWidgetCount=2, parentClass=/Script/FortniteUI.AthenaObjectiveTaskEntryBase, rootWidgetClass=&SizeBox, widgetCount=23}
 widgets [
  SizeBox_WidthOverride {widgetClassPath=&SizeBox, widget=SizeBox_WidthOverride, properties {bOverride_WidthOverride=true, widthOverride=400}}
  Border_Disable {
   widgetClassPath=&Border
   bIsVariable=true
   parent=SizeBox_WidthOverride
   widget=Border_Disable
   properties {
    background {
     bIsDynamicallyLoaded=false
     drawAs=NoDrawType
     imageType=NoImage
     mirroring=NoMirror
     resourceName=None
     resourceObject=null
     tiling=NoTile
     imageSize=<32 32>
     margin=<ltrb 0 0 0 0>
     outlineSettings {
      bUseBrushTransparency=false
      roundingType=HalfHeightRadius
      width=0
      color {colorUseRule=UseColor_Specified, specifiedColor=<rgba 0 0 0 0>}
      cornerRadii=<0 0 0 0>
     }
     tintColor {colorUseRule=UseColor_Specified, specifiedColor=<rgba 1 1 1 1>}
     uVRegion {bIsValid=false, max=<0 0>, min=<0 0>}
    }
    padding=<ltrb 0 0 0 0>
   }
   slot {class=SizeBoxSlot}
  }
  Overlay_EntryContent {
   widgetClassPath=&Overlay
   parent=Border_Disable
   widget=Overlay_EntryContent
   slot {class=BorderSlot, properties {padding=<ltrb 0 0 0 0>}}
  }
  Image_BG {
   widgetClassPath=&Image
   parent=Overlay_EntryContent
   widget=Image_BG
   properties {
    visibility=SelfHitTestInvisible
    brush {
     bIsDynamicallyLoaded=false
     drawAs=Box
     imageType=FullColor
     mirroring=NoMirror
     resourceName=None
     resourceObject=/Game/Athena/HUD/Discoverability/materials/MI_UI_Objectives_TitleBgBright
     tiling=NoTile
     imageSize=<32 24>
     margin=<ltrb 0.5 0.5 0.5 0.5>
     outlineSettings {
      bUseBrushTransparency=false
      roundingType=HalfHeightRadius
      width=0
      color {colorUseRule=UseColor_Specified, specifiedColor=<rgba 0 0 0 0>}
      cornerRadii=<0 0 0 0>
     }
     tintColor {colorUseRule=UseColor_Specified, specifiedColor=<rgba 1 1 1 1>}
     uVRegion {bIsValid=false, max=<0 0>, min=<0 0>}
    }
   }
   slot {class=OverlaySlot, properties {horizontalAlignment=HAlign_Fill, verticalAlignment=VAlign_Fill, padding=<ltrb 0 0 -4 0>}}
  }
  HorizontalBox_207 {
   widgetClassPath=&HorizontalBox
   parent=Overlay_EntryContent
   widget=HorizontalBox_207
   slot {class=OverlaySlot, properties {horizontalAlignment=HAlign_Fill, verticalAlignment=VAlign_Fill}}
  }
  CommonVisibilitySwitcher_0 {
   widgetClassPath=&CommonVisibilitySwitcher
   parent=HorizontalBox_207
   widget=CommonVisibilitySwitcher_0
   slot {class=HorizontalBoxSlot}
  }
  SizeBox_0 {
   widgetClassPath=&SizeBox
   parent=CommonVisibilitySwitcher_0
   widget=SizeBox_0
   properties {bOverride_HeightOverride=true, bOverride_WidthOverride=true, heightOverride=40, widthOverride=30}
   slot {class=CommonVisibilitySwitcherSlot}
  }
  Overlay_40 {
   widgetClassPath=&Overlay
   parent=SizeBox_0
   widget=Overlay_40
   slot {class=SizeBoxSlot, properties {horizontalAlignment=HAlign_Center, verticalAlignment=VAlign_Center}}
  }
  Image_RewardBG_1 {
   widgetClassPath=&Image
   bIsVariable=true
   parent=Overlay_40
   widget=Image_RewardBG_1
   properties {
    visibility=SelfHitTestInvisible
    brush {
     bIsDynamicallyLoaded=false
     drawAs=Image
     imageType=FullColor
     mirroring=NoMirror
     resourceName=None
     resourceObject=/Game/Athena/UI/Materials/M_UI_Circle
     tiling=NoTile
     imageSize=<24 24>
     margin=<ltrb 0.5 0.5 0.5 0.5>
     outlineSettings {
      bUseBrushTransparency=false
      roundingType=HalfHeightRadius
      width=0
      color {colorUseRule=UseColor_Specified, specifiedColor=<rgba 0 0 0 0>}
      cornerRadii=<0 0 0 0>
     }
     tintColor {colorUseRule=UseColor_Specified, specifiedColor=<rgba 1 1 1 0>}
     uVRegion {bIsValid=false, max=<0 0>, min=<0 0>}
    }
   }
   slot {class=OverlaySlot, properties {verticalAlignment=VAlign_Center}}
  }
  Text_ObjectIndex {
   widgetClassPath=/Script/FortniteUI.FortRichTextBlockKeybind
   bIsVariable=true
   parent=Overlay_40
   widget=Text_ObjectIndex
   properties {
    bOverrideDefaultStyle=true
    defaultTextStyleOverrideClass=&TextStyle-Base_C
    text="<TopCount>0.</>"
    textStyleSet=/Game/Athena/HUD/Discoverability/UrgentQuestRichTextStyles
    defaultTextStyleOverride {
     overflowPolicy=Clip
     transformPolicy=None
     colorAndOpacity {colorUseRule=UseColor_Specified, specifiedColor=<rgba 1 1 1 1>}
     font {
      bForceMonospaced=false
      bMaterialIsStencil=false
      fontMaterial=null
      fontObject=&NotoSans
      letterSpacing=0
      monospacedWidth=1
      size=24
      skewAmount=0
      typefaceFontName=NotoSans-Regular
      outlineSettings {
       bApplyOutlineToDropShadows=false
       bMiteredCorners=false
       bSeparateFillAlpha=false
       outlineMaterial=null
       outlineSize=0
       outlineColor=<rgba 0 0 0 1>
      }
     }
     highlightColor {colorUseRule=UseColor_Specified, specifiedColor=<rgba 0 0 0 1>}
     highlightShape {
      bIsDynamicallyLoaded=false
      drawAs=Image
      imageType=NoImage
      mirroring=NoMirror
      resourceName=None
      resourceObject=null
      tiling=NoTile
      imageSize=<32 32>
      margin=<ltrb 0 0 0 0>
      outlineSettings {
       bUseBrushTransparency=false
       roundingType=HalfHeightRadius
       width=0
       color {colorUseRule=UseColor_Specified, specifiedColor=<rgba 0 0 0 0>}
       cornerRadii=<0 0 0 0>
      }
      tintColor {colorUseRule=UseColor_Specified, specifiedColor=<rgba 1 1 1 1>}
      uVRegion {bIsValid=false, max=<0 0>, min=<0 0>}
     }
     selectedBackgroundColor {colorUseRule=UseColor_Foreground, specifiedColor=<rgba 1 0 1 1>}
     shadowColorAndOpacity=<rgba 0 0 0 1>
     shadowOffset=<0 0>
     strikeBrush {
      bIsDynamicallyLoaded=false
      drawAs=Image
      imageType=NoImage
      mirroring=NoMirror
      resourceName=None
      resourceObject=null
      tiling=NoTile
      imageSize=<32 32>
      margin=<ltrb 0 0 0 0>
      outlineSettings {
       bUseBrushTransparency=false
       roundingType=HalfHeightRadius
       width=0
       color {colorUseRule=UseColor_Specified, specifiedColor=<rgba 0 0 0 0>}
       cornerRadii=<0 0 0 0>
      }
      tintColor {colorUseRule=UseColor_Specified, specifiedColor=<rgba 1 1 1 1>}
      uVRegion {bIsValid=false, max=<0 0>, min=<0 0>}
     }
     underlineBrush {
      bIsDynamicallyLoaded=false
      drawAs=Image
      imageType=NoImage
      mirroring=NoMirror
      resourceName=None
      resourceObject=null
      tiling=NoTile
      imageSize=<32 32>
      margin=<ltrb 0 0 0 0>
      outlineSettings {
       bUseBrushTransparency=false
       roundingType=HalfHeightRadius
       width=0
       color {colorUseRule=UseColor_Specified, specifiedColor=<rgba 0 0 0 0>}
       cornerRadii=<0 0 0 0>
      }
      tintColor {colorUseRule=UseColor_Specified, specifiedColor=<rgba 1 1 1 1>}
      uVRegion {bIsValid=false, max=<0 0>, min=<0 0>}
     }
    }
   }
   slot {class=OverlaySlot, properties {horizontalAlignment=HAlign_Center, verticalAlignment=VAlign_Center}}
  }
  Overlay_Reward {
   widgetClassPath=&Overlay
   bIsVariable=true
   parent=CommonVisibilitySwitcher_0
   widget=Overlay_Reward
   properties {visibility=Collapsed}
   slot {class=CommonVisibilitySwitcherSlot}
  }
  Image_RewardBG {
   widgetClassPath=&Image
   bIsVariable=true
   parent=Overlay_Reward
   widget=Image_RewardBG
   properties {
    visibility=SelfHitTestInvisible
    brush {
     bIsDynamicallyLoaded=false
     drawAs=Box
     imageType=NoImage
     mirroring=NoMirror
     resourceName=None
     resourceObject=/Game/Athena/HUD/Discoverability/materials/MI_UI_Objectives_AdditionalReward_BG
     tiling=NoTile
     imageSize=<24 24>
     margin=<ltrb 0.5 0.5 0.5 0.5>
     outlineSettings {
      bUseBrushTransparency=false
      roundingType=HalfHeightRadius
      width=0
      color {colorUseRule=UseColor_Specified, specifiedColor=<rgba 0 0 0 0>}
      cornerRadii=<0 0 0 0>
     }
     tintColor {colorUseRule=UseColor_Specified, specifiedColor=<rgba 1 1 1 0.5>}
     uVRegion {bIsValid=false, max=<0 0>, min=<0 0>}
    }
   }
   slot {class=OverlaySlot, properties {horizontalAlignment=HAlign_Fill, verticalAlignment=VAlign_Fill, padding=<ltrb 0 0 -10 0>}}
  }
  Text_ObjectiveAdditionalReward {
   widgetClassPath=&FortRichTextBlockKeybind
   bIsVariable=true
   parent=Overlay_Reward
   widget=Text_ObjectiveAdditionalReward
   properties {
    bOverrideDefaultStyle=true
    keybindWidgetClass=/FortUILibrary/Input/KeybindWidgetDark_Small.KeybindWidgetDark_Small_C
    mobileTextBlockScale=1.2
    scrollStyle=/Game/Athena/Prototype/Blueprints/Galileo/UI/TextScrollStyle-Base-Ticker.TextScrollStyle-Base-Ticker_C
    text="<img id=\"BarsIcon\" width=\"22\"></> 0"
    textTransformPolicy=ToUpper
    visibility=Collapsed
    decoratorClasses [/Game/Athena/HUD/Discoverability/UrgentQuestRichTextDecorators.UrgentQuestRichTextDecorators_C]
    defaultTextStyleOverride {
     overflowPolicy=Clip
     transformPolicy=None
     colorAndOpacity {colorUseRule=UseColor_Specified, specifiedColor=<rgba 1 1 1 1>}
     font {
      bForceMonospaced=false
      bMaterialIsStencil=false
      fontMaterial=null
      fontObject=&BurbankSmall
      letterSpacing=0
      monospacedWidth=1
      size=12
      skewAmount=0.1
      typefaceFontName=Bold
      outlineSettings {
       bApplyOutlineToDropShadows=false
       bMiteredCorners=false
       bSeparateFillAlpha=false
       outlineMaterial=null
       outlineSize=2
       outlineColor=<rgba 0 0 0 0.25>
      }
     }
     highlightColor {colorUseRule=UseColor_Specified, specifiedColor=<rgba 0 0 0 0>}
     highlightShape {
      bIsDynamicallyLoaded=false
      drawAs=Image
      imageType=NoImage
      mirroring=NoMirror
      resourceName=None
      resourceObject=null
      tiling=NoTile
      imageSize=<32 32>
      margin=<ltrb 0 0 0 0>
      outlineSettings {
       bUseBrushTransparency=false
       roundingType=HalfHeightRadius
       width=0
       color {colorUseRule=UseColor_Specified, specifiedColor=<rgba 0 0 0 0>}
       cornerRadii=<0 0 0 0>
      }
      tintColor {colorUseRule=UseColor_Specified, specifiedColor=<rgba 1 1 1 1>}
      uVRegion {bIsValid=false, max=<0 0>, min=<0 0>}
     }
     selectedBackgroundColor {colorUseRule=UseColor_Foreground, specifiedColor=<rgba 1 0 1 1>}
     shadowColorAndOpacity=<rgba 0 0 0 0>
     shadowOffset=<0 0>
     strikeBrush {
      bIsDynamicallyLoaded=false
      drawAs=Image
      imageType=NoImage
      mirroring=NoMirror
      resourceName=None
      resourceObject=null
      tiling=NoTile
      imageSize=<32 32>
      margin=<ltrb 0 0 0 0>
      outlineSettings {
       bUseBrushTransparency=false
       roundingType=HalfHeightRadius
       width=0
       color {colorUseRule=UseColor_Specified, specifiedColor=<rgba 0 0 0 0>}
       cornerRadii=<0 0 0 0>
      }
      tintColor {colorUseRule=UseColor_Specified, specifiedColor=<rgba 1 1 1 1>}
      uVRegion {bIsValid=false, max=<0 0>, min=<0 0>}
     }
     underlineBrush {
      bIsDynamicallyLoaded=false
      drawAs=Image
      imageType=NoImage
      mirroring=NoMirror
      resourceName=None
      resourceObject=null
      tiling=NoTile
      imageSize=<32 32>
      margin=<ltrb 0 0 0 0>
      outlineSettings {
       bUseBrushTransparency=false
       roundingType=HalfHeightRadius
       width=0
       color {colorUseRule=UseColor_Specified, specifiedColor=<rgba 0 0 0 0>}
       cornerRadii=<0 0 0 0>
      }
      tintColor {colorUseRule=UseColor_Specified, specifiedColor=<rgba 1 1 1 1>}
      uVRegion {bIsValid=false, max=<0 0>, min=<0 0>}
     }
    }
    desiredKeybindWidgetSize=<16 16>
    renderTransformPivot=<0.5 0>
   }
   slot {class=OverlaySlot, properties {verticalAlignment=VAlign_Center, padding=<ltrb 5 0 0 0>}}
  }
  CommonVisibilitySwitcher_2 {
   widgetClassPath=&CommonVisibilitySwitcher
   parent=HorizontalBox_207
   widget=CommonVisibilitySwitcher_2
   slot {class=HorizontalBoxSlot, properties {verticalAlignment=VAlign_Center, padding=<ltrb 3 2 0 2>, size {sizeRule=Fill, value=1}}}
  }
  Text_ObjectiveDescriptionText {
   widgetClassPath=&FortRichTextBlockKeybind
   bInherited=true
   bIsVariable=true
   parent=CommonVisibilitySwitcher_2
   widget=Text_ObjectiveDescriptionText
   properties {
    autoWrapText=true
    bIsScrollingEnabled=false
    bOverrideDefaultStyle=true
    keybindWidgetClass=&KeybindWidgetDark_Small_C
    mobileTextBlockScale=1.2
    text=Collect Ammo Very Long Text Block
    decoratorClasses [&UrgentQuestRichTextDecorators_C]
    defaultTextStyleOverride {
     overflowPolicy=Clip
     transformPolicy=None
     colorAndOpacity {colorUseRule=UseColor_Specified, specifiedColor=<rgba 1 1 1 1>}
     font {
      bForceMonospaced=false
      bMaterialIsStencil=false
      fontMaterial=null
      fontObject=&HeadingNow
      letterSpacing=-50
      monospacedWidth=1
      size=10.5
      skewAmount=0
      typefaceFontName=75 Medium
      outlineSettings {
       bApplyOutlineToDropShadows=false
       bMiteredCorners=false
       bSeparateFillAlpha=false
       outlineMaterial=null
       outlineSize=0
       outlineColor=<rgba 0 0 0 0.5>
      }
     }
     highlightColor {colorUseRule=UseColor_Specified, specifiedColor=<rgba 0 0 0 0>}
     highlightShape {
      bIsDynamicallyLoaded=false
      drawAs=Image
      imageType=NoImage
      mirroring=NoMirror
      resourceName=None
      resourceObject=null
      tiling=NoTile
      imageSize=<32 32>
      margin=<ltrb 0 0 0 0>
      outlineSettings {
       bUseBrushTransparency=false
       roundingType=HalfHeightRadius
       width=0
       color {colorUseRule=UseColor_Specified, specifiedColor=<rgba 0 0 0 0>}
       cornerRadii=<0 0 0 0>
      }
      tintColor {colorUseRule=UseColor_Specified, specifiedColor=<rgba 1 1 1 1>}
      uVRegion {bIsValid=false, max=<0 0>, min=<0 0>}
     }
     selectedBackgroundColor {colorUseRule=UseColor_Foreground, specifiedColor=<rgba 1 0 1 1>}
     shadowColorAndOpacity=<rgba 0 0 0 0>
     shadowOffset=<0 0>
     strikeBrush {
      bIsDynamicallyLoaded=false
      drawAs=Image
      imageType=NoImage
      mirroring=NoMirror
      resourceName=None
      resourceObject=null
      tiling=NoTile
      imageSize=<32 32>
      margin=<ltrb 0 0 0 0>
      outlineSettings {
       bUseBrushTransparency=false
       roundingType=HalfHeightRadius
       width=0
       color {colorUseRule=UseColor_Specified, specifiedColor=<rgba 0 0 0 0>}
       cornerRadii=<0 0 0 0>
      }
      tintColor {colorUseRule=UseColor_Specified, specifiedColor=<rgba 1 1 1 1>}
      uVRegion {bIsValid=false, max=<0 0>, min=<0 0>}
     }
     underlineBrush {
      bIsDynamicallyLoaded=false
      drawAs=Image
      imageType=NoImage
      mirroring=NoMirror
      resourceName=None
      resourceObject=null
      tiling=NoTile
      imageSize=<32 32>
      margin=<ltrb 0 0 0 0>
      outlineSettings {
       bUseBrushTransparency=false
       roundingType=HalfHeightRadius
       width=0
       color {colorUseRule=UseColor_Specified, specifiedColor=<rgba 0 0 0 0>}
       cornerRadii=<0 0 0 0>
      }
      tintColor {colorUseRule=UseColor_Specified, specifiedColor=<rgba 1 1 1 1>}
      uVRegion {bIsValid=false, max=<0 0>, min=<0 0>}
     }
    }
    desiredKeybindWidgetSize=<16 16>
    renderTransformPivot=<0.5 0>
   }
   slot {class=CommonVisibilitySwitcherSlot}
  }
  UEFN_TextBlock_C_92 {
   widgetClassPath=&UEFN_TextBlock
   parent=CommonVisibilitySwitcher_2
   widget=UEFN_TextBlock_C_92
   properties {
    text=Complete
    font {
     bForceMonospaced=false
     bMaterialIsStencil=false
     fontMaterial=null
     fontObject=&BurbankSmall
     letterSpacing=-50
     monospacedWidth=1
     size=10.5
     skewAmount=0
     typefaceFontName=Bold
     outlineSettings {
      bApplyOutlineToDropShadows=false
      bMiteredCorners=false
      bSeparateFillAlpha=false
      outlineMaterial=null
      outlineSize=0
      outlineColor=<rgba 0 0 0 0.5>
     }
    }
   }
   slot {class=CommonVisibilitySwitcherSlot}
  }
  Switcher_Progress {
   widgetClassPath=&CommonActivatableWidgetSwitcher
   bIsVariable=true
   parent=HorizontalBox_207
   widget=Switcher_Progress
   slot {class=HorizontalBoxSlot}
  }
  ProgressWidget {
   widgetClassPath=&/WBP_Athena_Objectives_Progress
   bInherited=true
   bIsVariable=true
   parent=Switcher_Progress
   widget=ProgressWidget
   properties {include Checkmark in Succes=false}
   slot {class=WidgetSwitcherSlot}
  }
  SizeBox_Complete {
   widgetClassPath=&SizeBox
   parent=Switcher_Progress
   widget=SizeBox_Complete
   properties {bOverride_HeightOverride=true, bOverride_WidthOverride=true, heightOverride=30, minDesiredWidth=72, widthOverride=30}
   slot {class=WidgetSwitcherSlot, properties {horizontalAlignment=HAlign_Center, verticalAlignment=VAlign_Bottom, padding=<ltrb 0 0 0 5>}}
  }
  Overlay_2 {widgetClassPath=&Overlay, parent=SizeBox_Complete, widget=Overlay_2, slot {class=SizeBoxSlot}}
  Image_Check {
   widgetClassPath=&Image
   bIsVariable=true
   parent=Overlay_2
   widget=Image_Check
   properties {
    renderOpacity=0
    brush {
     bIsDynamicallyLoaded=false
     drawAs=Image
     imageType=FullColor
     mirroring=NoMirror
     resourceName=None
     resourceObject=/Game/UI/Foundation/Textures/Icons/Checks/T_UI_Icon_Check_MP_32
     tiling=NoTile
     imageSize=<32 32>
     margin=<ltrb 0 0 0 0>
     outlineSettings {
      bUseBrushTransparency=false
      roundingType=HalfHeightRadius
      width=0
      color {colorUseRule=UseColor_Specified, specifiedColor=<rgba 0 0 0 0>}
      cornerRadii=<0 0 0 0>
     }
     tintColor {colorUseRule=UseColor_Specified, specifiedColor=<rgba 1 1 1 1>}
     uVRegion {bIsValid=false, max=<0 0>, min=<0 0>}
    }
    colorAndOpacity=<rgba 0.0343 1 0.227 1>
   }
   slot {class=OverlaySlot, properties {horizontalAlignment=HAlign_Fill, verticalAlignment=VAlign_Fill}}
  }
  Image_Glow {
   widgetClassPath=&Image
   bIsVariable=true
   parent=Overlay_2
   widget=Image_Glow
   properties {
    renderOpacity=0
    brush {
     bIsDynamicallyLoaded=false
     drawAs=Image
     imageType=NoImage
     mirroring=NoMirror
     resourceName=None
     resourceObject=/Game/Athena/HUD/LTM/Wax/Materials/M_UI_TextGlow_Add
     tiling=NoTile
     imageSize=<130 130>
     margin=<ltrb 0 0 0 0>
     outlineSettings {
      bUseBrushTransparency=false
      roundingType=HalfHeightRadius
      width=0
      color {colorUseRule=UseColor_Specified, specifiedColor=<rgba 0 0 0 0>}
      cornerRadii=<0 0 0 0>
     }
     tintColor {colorUseRule=UseColor_Specified, specifiedColor=<rgba 0 1 0.1443 1>}
     uVRegion {bIsValid=false, max=<0 0>, min=<0 0>}
    }
   }
   slot {class=OverlaySlot, properties {horizontalAlignment=HAlign_Fill, verticalAlignment=VAlign_Fill}}
  }
  Image_Spark {
   widgetClassPath=&Image
   bIsVariable=true
   parent=Overlay_2
   widget=Image_Spark
   properties {
    renderOpacity=0
    brush {
     bIsDynamicallyLoaded=false
     drawAs=Image
     imageType=NoImage
     mirroring=NoMirror
     resourceName=None
     resourceObject=/Game/UI/InGame/Quest/Materials/M_UI_SparkStar
     tiling=NoTile
     imageSize=<120 116>
     margin=<ltrb 0 0 0 0>
     outlineSettings {
      bUseBrushTransparency=false
      roundingType=HalfHeightRadius
      width=0
      color {colorUseRule=UseColor_Specified, specifiedColor=<rgba 0 0 0 0>}
      cornerRadii=<0 0 0 0>
     }
     tintColor {colorUseRule=UseColor_Specified, specifiedColor=<rgba 1 1 1 1>}
     uVRegion {bIsValid=false, max=<0 0>, min=<0 0>}
    }
   }
   slot {class=OverlaySlot, properties {horizontalAlignment=HAlign_Fill, verticalAlignment=VAlign_Fill}}
  }
 ]
}
