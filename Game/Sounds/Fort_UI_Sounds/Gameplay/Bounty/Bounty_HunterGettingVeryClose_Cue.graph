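# Sound cue for the bounty UI warning that, by its name, signals a hunter getting very close to the player.
# A single wave player feeds the output directly, playing the HunterGettingCloser wave at reduced volume in the critical in-game UI sound class.
# A concurrency override allows one instance at a time, rejecting new plays and enforcing a retrigger delay to prevent spamming.
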SoundCue /Game/Sounds/Fort_UI_Sounds/Gameplay/Bounty/Bounty_HunterGettingVeryClose_Cue
root=WavePlayer_1
nodes [{id=WavePlayer_1, properties {soundWaveAssetPtr=&/Bounty_HunterGettingCloser}}]
properties {
 bOverrideConcurrency=true
 maxDistance=2097152
 soundClassObject=&Fort_UI_InGame_Critical
 volumeMultiplier=0.6
 concurrencyOverrides {maxCount=1, resolutionRule=PreventNew, retriggerTime=10}
}
