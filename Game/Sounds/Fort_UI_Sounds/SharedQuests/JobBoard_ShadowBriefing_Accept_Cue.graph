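# Sound cue for accepting a Shadow Briefing job on the shared-quest job board UI.
# A Mixer layers two one-shot waves at equal volume: JobBoard_Accept_UI_02 plays immediately, and JobBoard_Accept follows after a fixed short Delay.
# It routes through the non-critical in-game UI sound class.
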SoundCue /Game/Sounds/Fort_UI_Sounds/SharedQuests/JobBoard_ShadowBriefing_Accept_Cue
root=Mixer_0
nodes [
 {id=Delay_0, children [WavePlayer_0], properties {delayMax=0.3, delayMin=0.3}}
 {id=Mixer_0, children [WavePlayer_2,Delay_0], properties {inputVolume [1,1]}}
 {id=WavePlayer_0, properties {soundWaveAssetPtr=&/JobBoard_Accept}}
 {id=WavePlayer_2, properties {soundWaveAssetPtr=&/JobBoard_Accept_UI_02}}
]
properties {maxDistance=2097152, soundClassObject=&Fort_UI_InGame_NonCritical}
