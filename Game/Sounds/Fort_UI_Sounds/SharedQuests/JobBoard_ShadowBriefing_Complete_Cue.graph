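# JobBoard_ShadowBriefing_Complete_Cue is a UI sound cue, named for completing a Shadow Briefing on the Job Board in shared quests.
# A single wave player at the root plays the JobBoard_Completed sound wave, with no mixing, randomization or modulation nodes.
# It uses the non-critical in-game UI sound class and sets a very large max distance.
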SoundCue /Game/Sounds/Fort_UI_Sounds/SharedQuests/JobBoard_ShadowBriefing_Complete_Cue
root=WavePlayer_3
nodes [{id=WavePlayer_3, properties {soundWaveAssetPtr=&/JobBoard_Completed}}]
properties {maxDistance=2097152, soundClassObject=&Fort_UI_InGame_NonCritical}
