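# Sound cue for the Job Board shadow briefing failure in the shared quests UI.
# A single wave player plays the JobBoard_Complete_Whoosh wave into an Enveloper at the root.
# The Enveloper's pitch curve starts slightly below normal and falls over under a second at flat volume, turning the completion whoosh into a descending failure sound.
# The cue uses the non-critical in-game UI sound class.
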SoundCue /Game/Sounds/Fort_UI_Sounds/SharedQuests/JobBoard_ShadowBriefing_Failure_Cue
root=Enveloper_0
nodes [
 {
  id=Enveloper_0
  children [WavePlayer_1]
  properties {
   pitchCurve {editorCurveData {keys [{interpMode=RCIM_Cubic, tangentMode=RCTM_User, time=0, value=0.9}, <at 0.75 0.5>]}}
   volumeCurve {editorCurveData {keys [<at 0 1>]}}
  }
 }
 {id=WavePlayer_1, properties {soundWaveAssetPtr=&/JobBoard_Complete_Whoosh}}
]
properties {maxDistance=2097152, soundClassObject=&Fort_UI_InGame_NonCritical}
