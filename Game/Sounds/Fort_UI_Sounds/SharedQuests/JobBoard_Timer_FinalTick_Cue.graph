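# Sound cue for the final tick of the Job Board timer in the shared quests UI sounds.
# A single wave player feeds the root directly with a left-panned final timer boom, with no randomization or other processing nodes.
# It plays at slightly reduced volume and is routed to the Fort_Enemy_Damaged_PSM sound class rather than a UI-named one.
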SoundCue /Game/Sounds/Fort_UI_Sounds/SharedQuests/JobBoard_Timer_FinalTick_Cue
root=WavePlayer_1
nodes [{id=WavePlayer_1, properties {soundWaveAssetPtr=&/JobBoard_Final_Timer_Boom_LeftPanned}}]
properties {maxDistance=2097152, soundClassObject=/Game/Sounds/FortSoundClassesAndModes/Fort_Enemy_Damaged_PSM, volumeMultiplier=0.9}
