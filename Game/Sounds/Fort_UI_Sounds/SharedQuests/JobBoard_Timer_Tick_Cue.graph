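# Sound cue for the job board timer tick in the SharedQuests UI sounds.
# A single wave player feeds the left-panned tick wave straight to the output at reduced volume, with no randomization or looping.
# It uses the Creative small-toy focus attenuation and the Fort_Traps sound class.
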SoundCue /Game/Sounds/Fort_UI_Sounds/SharedQuests/JobBoard_Timer_Tick_Cue
root=WavePlayer_1
nodes [{id=WavePlayer_1, properties {soundWaveAssetPtr=&/Timer_Tick_LeftPanned}}]
properties {
 attenuationSettings=/Game/Sounds/Creative/Attenuations/Fort_Creative_SmallToy_Focus_Att
 maxDistance=2000
 soundClassObject=&Fort_Traps
 volumeMultiplier=0.6
}
